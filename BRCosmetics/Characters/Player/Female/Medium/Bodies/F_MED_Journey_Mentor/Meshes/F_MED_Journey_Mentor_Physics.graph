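# Physics asset for the F_MED_Journey_Mentor female player body cosmetic, defining collision and joints for its dangling skirt, belt, strap and sash bones.
# Pelvis and thigh capsules form the base: back skirts, right belt and front straps chain from the pelvis, while front skirts and sash attach to the thighs.
# Dynamic bones use capsules, except box-shaped belt bodies, with per-joint swing limits and twist locked on all but the sash.
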PhysicsAsset /BRCosmetics/Characters/Player/Female/Medium/Bodies/F_MED_Journey_Mentor/Meshes/F_MED_Journey_Mentor_Physics
bodies {
 dyn_L_backSkirt_2 [{shapeType=Capsule, length=7, radius=2.5, shapeName=dyn_L_backSkirt_2_capsule, center=<4 0 0>, rotation=<rot 90 0 0>}]
 dyn_L_backSkirt_3 [{shapeType=Capsule, length=7, radius=3, shapeName=dyn_L_backSkirt_2_capsule, center=<4 0 0>, rotation=<rot 90 0 0>}]
 dyn_L_frontSkirt_2 [
  {shapeType=Capsule, length=10, radius=3, shapeName=dyn_L_frontSkirt_2_capsule, center=<4 0 0>, rotation=<rot 90 0 0>}
  {shapeType=Capsule, length=10, radius=4, shapeName=dyn_L_frontSkirt_2_capsule, center=<4 6.5307 -11.4713>, rotation=<rot 90 0 0>}
 ]
 dyn_R_backSkirt_2 [{shapeType=Capsule, length=7, radius=2.5, shapeName=dyn_R_backSkirt_2_capsule, center=<-4 0 0>, rotation=<rot 90 0 0>}]
 dyn_R_backSkirt_3 [{shapeType=Capsule, length=7, radius=3, shapeName=dyn_R_backSkirt_2_capsule, center=<-4 0 0>, rotation=<rot 90 0 0>}]
 dyn_R_belt_B1 [{shapeType=Box, extentX=4.5, extentY=5.5, extentZ=7, shapeName=dyn_R_belt_A1_box, center=<3 0 0>, rotation=<rot 90 0 0>}]
 dyn_R_belt_B2 [{shapeType=Box, extentX=6, extentY=7.3, extentZ=10.5, shapeName=dyn_R_belt_A2_box, center=<3 0 0>, rotation=<rot 90 0 0>}]
 dyn_R_frontSkirt_2 [
  {shapeType=Capsule, length=10, radius=3, shapeName=dyn_L_frontSkirt_2_capsule, center=<-4 0 0>, rotation=<rot 90 0 0>}
  {shapeType=Capsule, length=10, radius=4, shapeName=dyn_R_frontSkirt_2_capsule, center=<-4 -6.694 10.781>, rotation=<rot 90 0 0>}
 ]
 dyn_front_strap_A_1 [{shapeType=Capsule, length=6, radius=3, shapeName=dyn_front_strap_A_1_capsule, center=<4 0 0>, rotation=<rot 90 0 0>}]
 dyn_front_strap_A_2 [{shapeType=Capsule, length=6, radius=3, shapeName=dyn_front_strap_A_2_capsule, center=<4 0 0>, rotation=<rot 90 0 0>}]
 dyn_front_strap_B_1 [{shapeType=Capsule, length=3, radius=2, shapeName=dyn_front_strap_B_1_capsule, center=<3 0 0>, rotation=<rot 90 0 0>}]
 dyn_front_strap_B_2 [{shapeType=Capsule, length=3, radius=2, shapeName=dyn_front_strap_B_2_capsule, center=<3 0 0>, rotation=<rot 90 0 0>}]
 dyn_sash_jnt_2 [{shapeType=Capsule, length=8, radius=3, shapeName=dyn_sash_jnt_2_capsule, center=<4 0 0>, rotation=<rot 90 0 0>}]
 dyn_sash_jnt_3 [{shapeType=Capsule, length=8, radius=5, shapeName=dyn_sash_jnt_3_capsule, center=<4 0 0>, rotation=<rot 90 0 0>}]
 pelvis [{shapeType=Capsule, length=5, radius=1, shapeName=pelvis_capsule, center=<-12.8401 0 0>}]
 thigh_l [{shapeType=Capsule, length=35, radius=10, shapeName=thigh_l_capsule, center=<-19 0 1.6868>, rotation=<rot 90 0 0>}]
 thigh_r [{shapeType=Capsule, length=40, radius=9, shapeName=thigh_r_capsule, center=<20.1717 0 -2.0019>, rotation=<rot 90 0 0>}]
}
constraints [
 {
  bone1Name=dyn_front_strap_A_2
  bone2Name=dyn_front_strap_A_1
  swing1LimitDegrees=10
  swing2LimitDegrees=30
  twistLimitDegrees=2
  twistMotion=Locked
 }
 {bone1Name=dyn_front_strap_B_1, bone2Name=pelvis, swing1LimitDegrees=30, swing2LimitDegrees=10, twistLimitDegrees=45, twistMotion=Locked}
 {
  bone1Name=dyn_front_strap_B_2
  bone2Name=dyn_front_strap_B_1
  swing1LimitDegrees=30
  swing2LimitDegrees=10
  twistLimitDegrees=45
  twistMotion=Locked
 }
 {bone1Name=dyn_sash_jnt_2, bone2Name=thigh_r, swing1LimitDegrees=20, swing2LimitDegrees=45, twistLimitDegrees=1}
 {bone1Name=dyn_sash_jnt_3, bone2Name=dyn_sash_jnt_2, swing1LimitDegrees=20, swing2LimitDegrees=60, twistLimitDegrees=1}
 {
  bone1Name=thigh_l
  bone2Name=pelvis
  swing1LimitDegrees=45
  swing1Motion=Free
  swing2LimitDegrees=45
  swing2Motion=Free
  twistLimitDegrees=45
  twistMotion=Free
 }
 {
  bone1Name=thigh_r
  bone2Name=pelvis
  swing1LimitDegrees=45
  swing1Motion=Free
  swing2LimitDegrees=45
  swing2Motion=Free
  twistLimitDegrees=45
  twistMotion=Free
 }
 {bone1Name=dyn_L_backSkirt_2, bone2Name=pelvis, swing1LimitDegrees=45, swing2LimitDegrees=45, twistLimitDegrees=45, twistMotion=Locked}
 {bone1Name=dyn_R_backSkirt_2, bone2Name=pelvis, swing1LimitDegrees=45, swing2LimitDegrees=45, twistLimitDegrees=45, twistMotion=Locked}
 {bone1Name=dyn_front_strap_A_1, bone2Name=pelvis, swing1LimitDegrees=10, swing2LimitDegrees=30, twistLimitDegrees=2, twistMotion=Locked}
 {bone1Name=dyn_R_frontSkirt_2, bone2Name=thigh_r, swing1LimitDegrees=30, swing2LimitDegrees=20, twistLimitDegrees=45, twistMotion=Locked}
 {bone1Name=dyn_L_frontSkirt_2, bone2Name=thigh_l, swing1LimitDegrees=45, swing2LimitDegrees=20, twistLimitDegrees=45, twistMotion=Locked}
 {
  bone1Name=dyn_L_backSkirt_3
  bone2Name=dyn_L_backSkirt_2
  swing1LimitDegrees=45
  swing2LimitDegrees=45
  twistLimitDegrees=45
  twistMotion=Locked
 }
 {
  bone1Name=dyn_R_backSkirt_3
  bone2Name=dyn_R_backSkirt_2
  swing1LimitDegrees=45
  swing2LimitDegrees=45
  twistLimitDegrees=45
  twistMotion=Locked
 }
 {bone1Name=dyn_R_belt_B1, bone2Name=pelvis, swing1LimitDegrees=20, swing2LimitDegrees=45, twistLimitDegrees=45, twistMotion=Locked}
 {bone1Name=dyn_R_belt_B2, bone2Name=dyn_R_belt_B1, swing1LimitDegrees=20, swing2LimitDegrees=45, twistLimitDegrees=45, twistMotion=Locked}
]
