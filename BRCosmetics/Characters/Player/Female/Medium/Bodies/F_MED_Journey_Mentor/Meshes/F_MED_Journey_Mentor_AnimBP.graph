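# Animation Blueprint for the female Journey Mentor cosmetic body, adding secondary motion to a pose copied from the parent mesh.
# A RigidBody sim and twist-deform Control Rig feed chained AnimDynamics, bone offsets and constraints for straps, danglers, bag, quiver, fur and belts.
# Event Graph turns movement and front-end idle montage state into blend weights, disabling bag, quiver, arrow and fur dynamics on ziplines and enabling right-belt dynamics while airborne.
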AnimBlueprint /BRCosmetics/Characters/Player/Female/Medium/Bodies/F_MED_Journey_Mentor/Meshes/F_MED_Journey_Mentor_AnimBP
target_skeleton=&/F_MED_Journey_Mentor_Skeleton
graphs [
 AnimGraph {nodes [
   n1 = RigidBody(
    ComponentPose=Local To Component(LocalPose=Copy Pose From Mesh() {bUseAttachedParent=true})
    Alpha=Get crouchRBANweight() -> (crouchRBANweight:double)
   ) {
    bClampLinearTranslationLimitToRefPose=true
    bOverrideWorldGravity=true
    overridePhysicsAsset=&/F_MED_Journey_Mentor_Physics
    simulationSpace=BaseBoneSpace
    alphaBoolBlend {blendOption=Linear}
    baseBoneRef {boneName=pelvis}
    componentAppliedLinearAccClamp=<3000 3000 3000>
    componentLinearAccScale=<0.3 0.3 0.3>
    componentLinearVelScale=<0.3 0.3 0.3>
    overrideWorldGravity {z=-980}
   }
   n2 = Control Rig(Source=Component To Local(ComponentPose=@n1)) {
    alphaBoolBlend {blendOption=Linear}
    controlRigAssetReference {blueprintRigClass=&/F_MED_Journey_Mentor_Deform_CtrlRig_twists.F_MED_Journey_Mentor_Deform_CtrlRig_twists_C}
   }
   n3 = "Transform (Modify) Bone - Bone: dyn_sash_jnt_1"(
    ComponentPose=Local To Component(LocalPose=@n2)
    Alpha=Get isNotAcceleratingStanding() -> (isNotAcceleratingStanding:double)
    Translation="0,0,0"
    Rotation="8,0,0"
    Scale="1,1,1"
   ) {
    rotationMode=BMM_Additive
    rotationSpace=BCS_BoneSpace
    alphaBoolBlend {blendOption=Linear}
    alphaScaleBiasClamp {bInterpResult=true}
    boneToModify {boneName=dyn_sash_jnt_1}
    rotation {pitch=8}
   }
   n4 = Anim Dynamics - Bone: dyn_L_leg_band_A_1(ComponentPose=@n3, Alpha=1, ExternalForce="0,0,0") {
    angularDampingOverride=0.9999
    angularSpringConstant=20
    bOverrideAngularDamping=true
    bUseGravityOverride=true
    gravityScale=0.35
    simulationSpace=RootRelative
    alphaBoolBlend {blendOption=Linear}
    boundBone {boneName=dyn_L_leg_band_A_1}
    componentLinearAccScale=<1 1 1>
    componentLinearVelScale=<1 1 1>
    gravityOverride {z=-980}
    physicsBodyDefinitions [{
      boundBone {boneName=dyn_L_leg_band_A_1}
      boxExtents=<8 8 8>
      constraintSetup {angularLimitsMax=<90 90 90>, angularLimitsMin=<-90 -90 -90>, angularTarget=<1 0 0>}
      localJointOffset=<1 0 0>
     }]
    relativeSpaceBone {boneName=lowerarm_l}
   }
   n5 = Anim Dynamics - Bone: dyn_L_leg_band_A_2(ComponentPose=@n4, Alpha=1, ExternalForce="0,0,0") {
    angularDampingOverride=0.9999
    angularSpringConstant=20
    bOverrideAngularDamping=true
    bUseGravityOverride=true
    gravityScale=0.35
    simulationSpace=RootRelative
    alphaBoolBlend {blendOption=Linear}
    boundBone {boneName=dyn_L_leg_band_A_2}
    componentLinearAccScale=<1 1 1>
    componentLinearVelScale=<1 1 1>
    gravityOverride {z=-980}
    physicsBodyDefinitions [{
      boundBone {boneName=dyn_L_leg_band_A_2}
      boxExtents=<8 8 8>
      constraintSetup {angularLimitsMax=<90 90 90>, angularLimitsMin=<-90 -90 -90>, angularTarget=<1 0 0>}
      localJointOffset=<1 0 0>
     }]
    relativeSpaceBone {boneName=lowerarm_l}
   }
   n6 = Anim Dynamics - Bone: dyn_L_leg_band_B_1(ComponentPose=@n5, Alpha=1, ExternalForce="0,0,0") {
    angularDampingOverride=0.9999
    angularSpringConstant=20
    bOverrideAngularDamping=true
    bUseGravityOverride=true
    gravityScale=0.35
    simulationSpace=RootRelative
    alphaBoolBlend {blendOption=Linear}
    boundBone {boneName=dyn_L_leg_band_B_1}
    componentLinearAccScale=<1 1 1>
    componentLinearVelScale=<1 1 1>
    gravityOverride {z=-980}
    physicsBodyDefinitions [{
      boundBone {boneName=dyn_L_leg_band_B_1}
      boxExtents=<8 8 8>
      constraintSetup {angularLimitsMax=<90 90 90>, angularLimitsMin=<-90 -90 -90>, angularTarget=<1 0 0>}
      localJointOffset=<1 0 0>
     }]
    relativeSpaceBone {boneName=lowerarm_r}
   }
   n7 = Anim Dynamics - Bone: dyn_L_leg_band_B_2(ComponentPose=@n6, Alpha=1, ExternalForce="0,0,0") {
    angularDampingOverride=0.9999
    angularSpringConstant=20
    bOverrideAngularDamping=true
    bUseGravityOverride=true
    gravityScale=0.35
    simulationSpace=RootRelative
    alphaBoolBlend {blendOption=Linear}
    boundBone {boneName=dyn_L_leg_band_B_2}
    componentLinearAccScale=<1 1 1>
    componentLinearVelScale=<1 1 1>
    gravityOverride {z=-980}
    physicsBodyDefinitions [{
      boundBone {boneName=dyn_L_leg_band_B_2}
      boxExtents=<8 8 8>
      constraintSetup {angularLimitsMax=<90 90 90>, angularLimitsMin=<-90 -90 -90>, angularTarget=<1 0 0>}
      localJointOffset=<1 0 0>
     }]
    relativeSpaceBone {boneName=lowerarm_r}
   }
   n8 = Anim Dynamics - Bone: dyn_R_leg_band_A_1(ComponentPose=@n7, Alpha=1, ExternalForce="0,0,0") {
    angularDampingOverride=0.9999
    angularSpringConstant=20
    bOverrideAngularDamping=true
    bUseGravityOverride=true
    gravityScale=0.35
    simulationSpace=RootRelative
    alphaBoolBlend {blendOption=Linear}
    boundBone {boneName=dyn_R_leg_band_A_1}
    componentLinearAccScale=<1 1 1>
    componentLinearVelScale=<1 1 1>
    gravityOverride {z=-980}
    physicsBodyDefinitions [{
      boundBone {boneName=dyn_R_leg_band_A_1}
      boxExtents=<8 8 8>
      constraintSetup {angularLimitsMax=<90 90 90>, angularLimitsMin=<-90 -90 -90>, angularTarget=<1 0 0>}
      localJointOffset=<1 0 0>
     }]
    relativeSpaceBone {boneName=lowerarm_l}
   }
   n9 = Anim Dynamics - Bone: dyn_R_leg_band_A_2(ComponentPose=@n8, Alpha=1, ExternalForce="0,0,0") {
    angularDampingOverride=0.9999
    angularSpringConstant=20
    bOverrideAngularDamping=true
    bUseGravityOverride=true
    gravityScale=0.35
    simulationSpace=RootRelative
    alphaBoolBlend {blendOption=Linear}
    boundBone {boneName=dyn_R_leg_band_A_2}
    componentLinearAccScale=<1 1 1>
    componentLinearVelScale=<1 1 1>
    gravityOverride {z=-980}
    physicsBodyDefinitions [{
      boundBone {boneName=dyn_R_leg_band_A_2}
      boxExtents=<8 8 8>
      constraintSetup {angularLimitsMax=<90 90 90>, angularLimitsMin=<-90 -90 -90>, angularTarget=<1 0 0>}
      localJointOffset=<1 0 0>
     }]
    relativeSpaceBone {boneName=lowerarm_l}
   }
   n10 = Anim Dynamics - Bone: dyn_R_leg_band_B_1(ComponentPose=@n9, Alpha=1, ExternalForce="0,0,0") {
    angularDampingOverride=0.9999
    angularSpringConstant=20
    bOverrideAngularDamping=true
    bUseGravityOverride=true
    gravityScale=0.35
    simulationSpace=RootRelative
    alphaBoolBlend {blendOption=Linear}
    boundBone {boneName=dyn_R_leg_band_B_1}
    componentLinearAccScale=<1 1 1>
    componentLinearVelScale=<1 1 1>
    gravityOverride {z=-980}
    physicsBodyDefinitions [{
      boundBone {boneName=dyn_R_leg_band_B_1}
      boxExtents=<8 8 8>
      constraintSetup {angularLimitsMax=<90 90 90>, angularLimitsMin=<-90 -90 -90>, angularTarget=<1 0 0>}
      localJointOffset=<1 0 0>
     }]
    relativeSpaceBone {boneName=lowerarm_r}
   }
   n11 = Anim Dynamics - Bone: dyn_R_leg_band_B_2(ComponentPose=@n10, Alpha=1, ExternalForce="0,0,0") {
    angularDampingOverride=0.9999
    angularSpringConstant=20
    bOverrideAngularDamping=true
    bUseGravityOverride=true
    gravityScale=0.35
    simulationSpace=RootRelative
    alphaBoolBlend {blendOption=Linear}
    boundBone {boneName=dyn_R_leg_band_B_2}
    componentLinearAccScale=<1 1 1>
    componentLinearVelScale=<1 1 1>
    gravityOverride {z=-980}
    physicsBodyDefinitions [{
      boundBone {boneName=dyn_R_leg_band_B_2}
      boxExtents=<8 8 8>
      constraintSetup {angularLimitsMax=<90 90 90>, angularLimitsMin=<-90 -90 -90>, angularTarget=<1 0 0>}
      localJointOffset=<1 0 0>
     }]
    relativeSpaceBone {boneName=lowerarm_r}
   }
   n12 = Anim Dynamics - Bone: L_armDanglers_1(ComponentPose=@n11, Alpha=1, ExternalForce="0,0,0") {
    angularDampingOverride=0.9999
    angularSpringConstant=20
    bAngularSpring=true
    bOverrideAngularDamping=true
    gravityScale=0.35
    simulationSpace=RootRelative
    alphaBoolBlend {blendOption=Linear}
    boundBone {boneName=L_armDanglers_1}
    componentLinearAccScale=<1 1 1>
    componentLinearVelScale=<1 1 1>
    gravityOverride {z=-980}
    physicsBodyDefinitions [{
      boundBone {boneName=L_armDanglers_1}
      boxExtents=<6 6 6>
      constraintSetup {angularLimitsMax=<180 180 180>, angularLimitsMin=<-180 -180 -180>, angularTarget=<1 0 0>}
      localJointOffset=<-2 0 0>
     }]
    relativeSpaceBone {boneName=lowerarm_l}
   }
   n13 = Anim Dynamics - Bone: L_armDanglers_2(ComponentPose=@n12, Alpha=1, ExternalForce="0,0,0") {
    angularDampingOverride=0.9999
    angularSpringConstant=20
    bAngularSpring=true
    bOverrideAngularDamping=true
    gravityScale=0.35
    simulationSpace=RootRelative
    alphaBoolBlend {blendOption=Linear}
    boundBone {boneName=L_armDanglers_2}
    componentLinearAccScale=<1 1 1>
    componentLinearVelScale=<1 1 1>
    gravityOverride {z=-980}
    physicsBodyDefinitions [{
      boundBone {boneName=L_armDanglers_2}
      boxExtents=<8 8 8>
      constraintSetup {angularLimitsMax=<180 180 180>, angularLimitsMin=<-180 -180 -180>, angularTarget=<1 0 0>}
      localJointOffset=<-2 0 0>
     }]
    relativeSpaceBone {boneName=lowerarm_l}
   }
   n14 = Anim Dynamics - Bone: R_armDanglers_1(ComponentPose=@n13, Alpha=1, ExternalForce="0,0,0") {
    angularDampingOverride=0.9999
    angularSpringConstant=20
    bAngularSpring=true
    bOverrideAngularDamping=true
    gravityScale=0.35
    simulationSpace=RootRelative
    alphaBoolBlend {blendOption=Linear}
    boundBone {boneName=R_armDanglers_1}
    componentLinearAccScale=<1 1 1>
    componentLinearVelScale=<1 1 1>
    gravityOverride {z=-980}
    physicsBodyDefinitions [{
      boundBone {boneName=R_armDanglers_1}
      boxExtents=<6 6 6>
      constraintSetup {angularLimitsMax=<180 180 180>, angularLimitsMin=<-180 -180 -180>, angularTarget=<1 0 0>}
      localJointOffset=<2 0 0>
     }]
    relativeSpaceBone {boneName=lowerarm_r}
   }
   n15 = Anim Dynamics - Bone: R_armDanglers_2(ComponentPose=@n14, Alpha=1, ExternalForce="0,0,0") {
    angularDampingOverride=0.9999
    angularSpringConstant=20
    bAngularSpring=true
    bOverrideAngularDamping=true
    gravityScale=0.35
    simulationSpace=RootRelative
    alphaBoolBlend {blendOption=Linear}
    boundBone {boneName=R_armDanglers_2}
    componentLinearAccScale=<1 1 1>
    componentLinearVelScale=<1 1 1>
    gravityOverride {z=-980}
    physicsBodyDefinitions [{
      boundBone {boneName=R_armDanglers_2}
      boxExtents=<8 8 8>
      constraintSetup {angularLimitsMax=<180 180 180>, angularLimitsMin=<-180 -180 -180>, angularTarget=<1 0 0>}
      localJointOffset=<2 0 0>
     }]
    relativeSpaceBone {boneName=lowerarm_r}
   }
   n16 = Get ziplineDynamics() -> (ziplineDynamics:double)
   n17 = Anim Dynamics - Bone: dyn_backBag(ComponentPose=@n15, Alpha=@n16, ExternalForce="0,0,0") {
    angularDampingOverride=0.9999
    angularSpringConstant=2000
    bAngularSpring=true
    bOverrideAngularDamping=true
    bUseGravityOverride=true
    gravityScale=0.35
    simulationSpace=RootRelative
    alphaBoolBlend {blendOption=Linear}
    boundBone {boneName=dyn_backBag}
    componentLinearAccScale=<1 1 1>
    componentLinearVelScale=<1 1 1>
    gravityOverride {z=-2000}
    physicsBodyDefinitions [{
      boundBone {boneName=dyn_backBag}
      boxExtents=<30 30 30>
      constraintSetup {angularLimitsMax=<0 10 5>, angularLimitsMin=<0 -10 -45>, angularTarget=<1 0 0>}
      localJointOffset=<3 0 0>
     }]
    relativeSpaceBone {boneName=lowerarm_l}
   }
   n18 = Anim Dynamics - Bone: dyn_quiver_jnt_master(ComponentPose=@n17, Alpha=@n16, ExternalForce="0,0,0") {
    angularDampingOverride=0.9999
    angularSpringConstant=100
    bAngularSpring=true
    bOverrideAngularDamping=true
    bUseGravityOverride=true
    gravityScale=0.35
    alphaBoolBlend {blendOption=Linear}
    boundBone {boneName=dyn_quiver_jnt_master}
    componentLinearAccScale=<1 1 1>
    componentLinearVelScale=<1 1 1>
    gravityOverride {z=-980}
    physicsBodyDefinitions [{
      boundBone {boneName=dyn_quiver_jnt_master}
      boxExtents=<15 15 15>
      constraintSetup {angularLimitsMax=<0 5 0>, angularLimitsMin=<0 -45 0>, angularTarget=<1 0 0>}
      localJointOffset=<0 0 0.5>
     }]
    relativeSpaceBone {boneName=lowerarm_l}
   }
   n19 = Anim Dynamics - Bone: dyn_arrow_1(ComponentPose=@n18, Alpha=@n16, ExternalForce="0,0,0") {
    angularDampingOverride=0.9999
    angularSpringConstant=300
    bAngularSpring=true
    bOverrideAngularDamping=true
    gravityScale=0.35
    simulationSpace=RootRelative
    alphaBoolBlend {blendOption=Linear}
    boundBone {boneName=dyn_arrow_1}
    componentLinearAccScale=<1 1 1>
    componentLinearVelScale=<1 1 1>
    gravityOverride {z=-980}
    physicsBodyDefinitions [{
      boundBone {boneName=dyn_arrow_1}
      boxExtents=<15 15 15>
      constraintSetup {angularLimitsMax=<0 2 2>, angularLimitsMin=<0 -2 -2>, angularTarget=<1 0 0>}
      localJointOffset=<1 0 0>
     }]
    relativeSpaceBone {boneName=lowerarm_l}
   }
   n20 = Anim Dynamics - Bone: dyn_arrow_2(ComponentPose=@n19, Alpha=@n16, ExternalForce="0,0,0") {
    angularDampingOverride=0.9999
    angularSpringConstant=300
    bAngularSpring=true
    bOverrideAngularDamping=true
    gravityScale=0.35
    simulationSpace=RootRelative
    alphaBoolBlend {blendOption=Linear}
    boundBone {boneName=dyn_arrow_2}
    componentLinearAccScale=<1 1 1>
    componentLinearVelScale=<1 1 1>
    gravityOverride {z=-980}
    physicsBodyDefinitions [{
      boundBone {boneName=dyn_arrow_2}
      boxExtents=<15 15 15>
      constraintSetup {angularLimitsMax=<0 2 2>, angularLimitsMin=<0 -2 -2>, angularTarget=<1 0 0>}
      localJointOffset=<1 0 0>
     }]
    relativeSpaceBone {boneName=lowerarm_l}
   }
   n21 = Anim Dynamics - Bone: dyn_arrow_3(ComponentPose=@n20, Alpha=@n16, ExternalForce="0,0,0") {
    angularDampingOverride=0.9999
    angularSpringConstant=300
    bAngularSpring=true
    bOverrideAngularDamping=true
    gravityScale=0.35
    simulationSpace=RootRelative
    alphaBoolBlend {blendOption=Linear}
    boundBone {boneName=dyn_arrow_3}
    componentLinearAccScale=<1 1 1>
    componentLinearVelScale=<1 1 1>
    gravityOverride {z=-980}
    physicsBodyDefinitions [{
      boundBone {boneName=dyn_arrow_3}
      boxExtents=<15 15 15>
      constraintSetup {angularLimitsMax=<0 2 2>, angularLimitsMin=<0 -2 -2>, angularTarget=<1 0 0>}
      localJointOffset=<1 0 0>
     }]
    relativeSpaceBone {boneName=lowerarm_l}
   }
   n22 = Anim Dynamics - Bone: dyn_arrow_4(ComponentPose=@n21, Alpha=@n16, ExternalForce="0,0,0") {
    angularDampingOverride=0.9999
    angularSpringConstant=300
    bAngularSpring=true
    bOverrideAngularDamping=true
    gravityScale=0.35
    simulationSpace=RootRelative
    alphaBoolBlend {blendOption=Linear}
    boundBone {boneName=dyn_arrow_4}
    componentLinearAccScale=<1 1 1>
    componentLinearVelScale=<1 1 1>
    gravityOverride {z=-980}
    physicsBodyDefinitions [{
      boundBone {boneName=dyn_arrow_4}
      boxExtents=<15 15 15>
      constraintSetup {angularLimitsMax=<0 2 2>, angularLimitsMin=<0 -2 -2>, angularTarget=<1 0 0>}
      localJointOffset=<1 0 0>
     }]
    relativeSpaceBone {boneName=lowerarm_l}
   }
   n23 = Anim Dynamics - Bone: dyn_arrow_5(ComponentPose=@n22, Alpha=@n16, ExternalForce="0,0,0") {
    angularDampingOverride=0.9999
    angularSpringConstant=300
    bAngularSpring=true
    bOverrideAngularDamping=true
    gravityScale=0.35
    simulationSpace=RootRelative
    alphaBoolBlend {blendOption=Linear}
    boundBone {boneName=dyn_arrow_5}
    componentLinearAccScale=<1 1 1>
    componentLinearVelScale=<1 1 1>
    gravityOverride {z=-980}
    physicsBodyDefinitions [{
      boundBone {boneName=dyn_arrow_5}
      boxExtents=<15 15 15>
      constraintSetup {angularLimitsMax=<0 2 2>, angularLimitsMin=<0 -2 -2>, angularTarget=<1 0 0>}
      localJointOffset=<1 0 0>
     }]
    relativeSpaceBone {boneName=lowerarm_l}
   }
   n24 = Anim Dynamics - Bone: dyn_R_fur_top(ComponentPose=@n23, Alpha=@n16, ExternalForce="0,0,0") {
    angularDampingOverride=1
    angularSpringConstant=20000
    bAngularSpring=true
    bLinearSpring=true
    bOverrideAngularDamping=true
    bOverrideLinearDamping=true
    bUseGravityOverride=true
    gravityScale=0.35
    linearDampingOverride=1
    linearSpringConstant=50
    simulationSpace=RootRelative
    alphaBoolBlend {blendOption=Linear}
    boundBone {boneName=dyn_R_fur_top}
    componentLinearAccScale=<0.6 0.6 0.6>
    componentLinearVelScale=<0.6 0.6 0.6>
    gravityOverride {z=-1400}
    physicsBodyDefinitions [{
      boundBone {boneName=dyn_R_fur_top}
      boxExtents=<80 80 80>
      constraintSetup {angularLimitsMax=<0 0 0>, angularLimitsMin=<0 0 0>, angularTarget=<0 0 0>, linearAxesMax=<1 1 3>, linearAxesMin=<-1 -1 0>}
      localJointOffset=<0 0 0>
     }]
   }
   n25 = Anim Dynamics - Bone: dyn_R_fur_back(ComponentPose=@n24, Alpha=@n16, ExternalForce="0,0,0") {
    angularDampingOverride=1
    angularSpringConstant=20000
    bAngularSpring=true
    bOverrideAngularDamping=true
    bOverrideLinearDamping=true
    bUseGravityOverride=true
    gravityScale=0.35
    linearDampingOverride=1
    linearSpringConstant=50
    simulationSpace=RootRelative
    alphaBoolBlend {blendOption=Linear}
    boundBone {boneName=dyn_R_fur_back}
    componentLinearAccScale=<0.6 0.6 0.6>
    componentLinearVelScale=<0.6 0.6 0.6>
    gravityOverride {z=-1400}
    physicsBodyDefinitions [{
      boundBone {boneName=dyn_R_fur_back}
      boxExtents=<80 80 80>
      constraintSetup {angularLimitsMax=<5 5 5>, angularLimitsMin=<-5 -5 -5>, angularTarget=<0 0 0>, linearAxesMax=<2 2 2>, linearAxesMin=<-2 -2 0>}
      localJointOffset=<0 0 0>
     }]
   }
   n26 = Anim Dynamics - Bone: dyn_L_fur_front(ComponentPose=@n25, Alpha=@n16, ExternalForce="0,0,0") {
    angularDampingOverride=1
    angularSpringConstant=20000
    bAngularSpring=true
    bOverrideAngularDamping=true
    bOverrideLinearDamping=true
    bUseGravityOverride=true
    gravityScale=0.35
    linearDampingOverride=1
    linearSpringConstant=50
    simulationSpace=RootRelative
    alphaBoolBlend {blendOption=Linear}
    boundBone {boneName=dyn_L_fur_front}
    componentLinearAccScale=<0.6 0.6 0.6>
    componentLinearVelScale=<0.6 0.6 0.6>
    gravityOverride {z=-1400}
    physicsBodyDefinitions [{
      boundBone {boneName=dyn_L_fur_front}
      boxExtents=<80 80 80>
      constraintSetup {angularLimitsMax=<5 5 5>, angularLimitsMin=<-5 -5 -5>, angularTarget=<0 0 0>, linearAxesMax=<2 3 4>, linearAxesMin=<-2 -3 -2>}
      localJointOffset=<0 0 0>
     }]
   }
   n27 = Anim Dynamics - Bone: dyn_L_fur_top(ComponentPose=@n26, Alpha=@n16, ExternalForce="0,0,0") {
    angularDampingOverride=1
    angularSpringConstant=20000
    bAngularSpring=true
    bLinearSpring=true
    bOverrideAngularDamping=true
    bOverrideLinearDamping=true
    bUseGravityOverride=true
    gravityScale=0.35
    linearDampingOverride=1
    linearSpringConstant=50
    simulationSpace=RootRelative
    alphaBoolBlend {blendOption=Linear}
    boundBone {boneName=dyn_L_fur_top}
    componentLinearAccScale=<0.6 0.6 0.6>
    componentLinearVelScale=<0.6 0.6 0.6>
    gravityOverride {z=-1400}
    physicsBodyDefinitions [{
      boundBone {boneName=dyn_L_fur_top}
      boxExtents=<80 80 80>
      constraintSetup {angularLimitsMax=<0 0 0>, angularLimitsMin=<0 0 0>, angularTarget=<0 0 0>, linearAxesMax=<1 1 3>, linearAxesMin=<-1 -1 0>}
      localJointOffset=<0 0 0>
     }]
   }
   n28 = Anim Dynamics - Bone: dyn_L_fur_back(ComponentPose=@n27, Alpha=@n16, ExternalForce="0,0,0") {
    angularDampingOverride=1
    angularSpringConstant=20000
    bAngularSpring=true
    bOverrideAngularDamping=true
    bOverrideLinearDamping=true
    bUseGravityOverride=true
    gravityScale=0.35
    linearDampingOverride=1
    linearSpringConstant=50
    simulationSpace=RootRelative
    alphaBoolBlend {blendOption=Linear}
    boundBone {boneName=dyn_L_fur_back}
    componentLinearAccScale=<0.6 0.6 0.6>
    componentLinearVelScale=<0.6 0.6 0.6>
    gravityOverride {z=-1400}
    physicsBodyDefinitions [{
      boundBone {boneName=dyn_L_fur_back}
      boxExtents=<80 80 80>
      constraintSetup {angularLimitsMax=<5 5 5>, angularLimitsMin=<-5 -5 -5>, angularTarget=<0 0 0>, linearAxesMax=<2 2 2>, linearAxesMin=<-2 -2 0>}
      localJointOffset=<0 0 0>
     }]
   }
   n29 = Get isSkydiving() -> (isSkydiving:double)
   n30 = "Transform (Modify) Bone - Bone: dyn_R_fur_back"(ComponentPose=@n28, Alpha=@n29, Translation="0,0,5", Rotation="0,0,0", Scale="1,1,1") {translationMode=BMM_Additive, alphaBoolBlend {blendOption=Linear}, boneToModify {boneName=dyn_R_fur_back}, translation {z=5}}
   n31 = "Transform (Modify) Bone - Bone: dyn_L_fur_back"(ComponentPose=@n30, Alpha=@n29, Translation="0,0,5", Rotation="0,0,0", Scale="1,1,1") {translationMode=BMM_Additive, alphaBoolBlend {blendOption=Linear}, boneToModify {boneName=dyn_L_fur_back}, translation {z=5}}
   n32 = "Transform (Modify) Bone - Bone: dyn_L_fur_top"(ComponentPose=@n31, Alpha=@n29, Translation="0,0,3", Rotation="0,0,0", Scale="1,1,1") {translationMode=BMM_Additive, alphaBoolBlend {blendOption=Linear}, boneToModify {boneName=dyn_L_fur_top}, translation {z=3}}
   n33 = "Transform (Modify) Bone - Bone: dyn_R_fur_top"(ComponentPose=@n32, Alpha=@n29, Translation="0,0,3", Rotation="0,0,0", Scale="1,1,1") {translationMode=BMM_Additive, alphaBoolBlend {blendOption=Linear}, boneToModify {boneName=dyn_R_fur_top}, translation {z=3}}
   n34 = "Constraint - dyn_L_vambraceUp (P)"(ComponentPose=@n33, Alpha=1, ConstraintWeights_0=1) {
    alphaBoolBlend {blendOption=Linear}
    boneToModify {boneName=dyn_L_vambraceUp}
    constraintSetup [{targetBone {boneName=hand_l}}]
    constraintWeights [1]
   }
   n35 = "Constraint - dyn_R_vambraceUp (P)"(ComponentPose=@n34, Alpha=1, ConstraintWeights_0=1) {
    alphaBoolBlend {blendOption=Linear}
    boneToModify {boneName=dyn_R_vambraceUp}
    constraintSetup [{targetBone {boneName=hand_r}}]
    constraintWeights [1]
   }
   n36 = "Constraint - dyn_beltConstraint_jnt_B (PPPPPPPPPPPP)"!(
    ComponentPose=@n35
    Alpha=1
    ConstraintWeights_0=0.5365
    ConstraintWeights_1:float=0.102
    ConstraintWeights_2:float=0.1371
    ConstraintWeights_3:float=0.0294
    ConstraintWeights_4:float=0.0287
    ConstraintWeights_5:float=0.0054
    ConstraintWeights_6:float=0.0008
    ConstraintWeights_7:float=0.1155
    ConstraintWeights_8:float=0.0364
    ConstraintWeights_9:float=0.0036
    ConstraintWeights_10:float=0.0016
    ConstraintWeights_11:float=0.0031
   ) -> (Pose) {
    alphaBoolBlend {blendOption=Linear}
    boneToModify {boneName=dyn_beltConstraint_jnt_B}
    constraintSetup [
     {targetBone {boneName=pelvis}}
     {targetBone {boneName=spine_01}}
     {targetBone {boneName=spine_02}}
     {targetBone {boneName=spine_03}}
     {targetBone {boneName=spine_04}}
     {targetBone {boneName=spine_05}}
     {targetBone {boneName=clavicle_l}}
     {targetBone {boneName=thigh_l}}
     {targetBone {boneName=thigh_r}}
     {targetBone {boneName=dyn_L_frontSkirt_2}}
     {targetBone {boneName=dyn_R_frontSkirt_2}}
     {targetBone {boneName=upperarm_deformTwist_l}}
    ]
    constraintWeights [1,1,1,1,1,1,1,1,1,1,1,1]
   }
   n37 = "Constraint - dyn_pendant_jnt (PPPPPPP)"!(
    ComponentPose=@n36
    Alpha=0.7
    ConstraintWeights_0=0.0039
    ConstraintWeights_1:float=0.0344
    ConstraintWeights_2:float=0.5058
    ConstraintWeights_3:float=0.1871
    ConstraintWeights_4:float=0.0173
    ConstraintWeights_5:float=0.2503
    ConstraintWeights_6:float=0.0012
   ) -> (Pose) {
    alphaBoolBlend {blendOption=Linear}
    boneToModify {boneName=dyn_pendant_jnt}
    constraintSetup [
     {targetBone {boneName=pelvis}}
     {targetBone {boneName=spine_03}}
     {targetBone {boneName=spine_04}}
     {targetBone {boneName=spine_05}}
     {targetBone {boneName=dyn_R_fur_top}}
     {targetBone {boneName=dyn_R_shoulderpad_jnt}}
     {targetBone {boneName=upperarm_deformTwist_l}}
    ]
    constraintWeights [1,1,1,1,1,1,1]
   }
   n38 = "Constraint - dyn_sash_jnt_1 (P)"(ComponentPose=@n37, Alpha=0.3, ConstraintWeights_0=1) {
    alphaBoolBlend {blendOption=Linear}
    boneToModify {boneName=dyn_sash_jnt_1}
    constraintSetup [{targetBone {boneName=thigh_l}}]
    constraintWeights [1]
   }
   n39 = "Constraint - dyn_sash_jnt_2 (P)"(ComponentPose=@n38, Alpha=0.6, ConstraintWeights_0=1) {
    alphaBoolBlend {blendOption=Linear}
    alphaScaleBiasClamp {bInterpResult=true}
    boneToModify {boneName=dyn_sash_jnt_2}
    constraintSetup [{targetBone {boneName=dyn_R_frontSkirt_2}}]
    constraintWeights [1]
   }
   n40 = "Constraint - dyn_R_belt_B1 (P)"(ComponentPose=@n39, Alpha=0.8, ConstraintWeights_0=1) {
    alphaBoolBlend {blendOption=Linear}
    boneToModify {boneName=dyn_R_belt_B1}
    constraintSetup [{targetBone {boneName=thigh_r}}]
    constraintWeights [1]
   }
   n41 = "Transform (Modify) Bone - Bone: dyn_R_belt_A1"(ComponentPose=@n40, Alpha=1, Translation="0,0,0", Rotation="-5,0,0", Scale="1,1,1") {rotationMode=BMM_Additive, alphaBoolBlend {blendOption=Linear}, boneToModify {boneName=dyn_R_belt_A1}, rotation {pitch=-5}}
   n42 = "Transform (Modify) Bone - Bone: dyn_R_belt_B1"(ComponentPose=@n41, Alpha=1, Translation="0,0,0", Rotation="-5,0,0", Scale="1,1,1") {rotationMode=BMM_Additive, alphaBoolBlend {blendOption=Linear}, boneToModify {boneName=dyn_R_belt_B1}, rotation {pitch=-5}}
   n43 = Get isJumping() -> (isJumping:double)
   n44 = Anim Dynamics - Bone: dyn_R_belt_A1(ComponentPose=@n42, Alpha=@n43, ExternalForce="0,0,0") {
    angularDampingOverride=0.9999
    angularSpringConstant=2000
    bAngularSpring=true
    bOverrideAngularDamping=true
    bUseGravityOverride=true
    gravityScale=0.35
    simulationSpace=RootRelative
    alphaBoolBlend {blendOption=Linear}
    alphaScaleBiasClamp {bInterpResult=true}
    boundBone {boneName=dyn_R_belt_A1}
    componentLinearAccScale=<1 1 1>
    componentLinearVelScale=<1 1 1>
    gravityOverride {z=-2000}
    physicsBodyDefinitions [{
      boundBone {boneName=dyn_R_belt_A1}
      boxExtents=<30 30 30>
      constraintSetup {angularLimitsMax=<0 5 0>, angularLimitsMin=<0 -5 -10>, angularTarget=<1 0 0>}
      localJointOffset=<3 0 0>
     }]
    relativeSpaceBone {boneName=lowerarm_l}
   }
   n45 = Anim Dynamics - Bone: dyn_R_belt_A2(ComponentPose=@n44, Alpha=@n43, ExternalForce="0,0,0") {
    angularDampingOverride=0.9999
    angularSpringConstant=2000
    bAngularSpring=true
    bOverrideAngularDamping=true
    bUseGravityOverride=true
    gravityScale=0.35
    simulationSpace=RootRelative
    alphaBoolBlend {blendOption=Linear}
    alphaScaleBiasClamp {bInterpResult=true}
    boundBone {boneName=dyn_R_belt_A2}
    componentLinearAccScale=<1 1 1>
    componentLinearVelScale=<1 1 1>
    gravityOverride {z=-2000}
    physicsBodyDefinitions [{
      boundBone {boneName=dyn_R_belt_A2}
      boxExtents=<30 30 30>
      constraintSetup {angularLimitsMax=<0 5 0>, angularLimitsMin=<0 -5 -10>, angularTarget=<1 0 0>}
      localJointOffset=<3 0 0>
     }]
    relativeSpaceBone {boneName=lowerarm_l}
   }
   n46 = Anim Dynamics - Bone: dyn_R_belt_B1(ComponentPose=@n45, Alpha=@n43, ExternalForce="0,0,0") {
    angularDampingOverride=0.9999
    angularSpringConstant=2000
    bAngularSpring=true
    bOverrideAngularDamping=true
    bUseGravityOverride=true
    gravityScale=0.35
    simulationSpace=RootRelative
    alphaBoolBlend {blendOption=Linear}
    alphaScaleBiasClamp {bInterpResult=true}
    boundBone {boneName=dyn_R_belt_B1}
    componentLinearAccScale=<1 1 1>
    componentLinearVelScale=<1 1 1>
    gravityOverride {z=-2000}
    physicsBodyDefinitions [{
      boundBone {boneName=dyn_R_belt_B1}
      boxExtents=<30 30 30>
      constraintSetup {angularLimitsMax=<0 5 0>, angularLimitsMin=<0 -5 -10>, angularTarget=<1 0 0>}
      localJointOffset=<3 0 0>
     }]
    relativeSpaceBone {boneName=lowerarm_l}
   }
   n47 = Anim Dynamics - Bone: dyn_R_belt_B2(ComponentPose=@n46, Alpha=@n43, ExternalForce="0,0,0") {
    angularDampingOverride=0.9999
    angularSpringConstant=2000
    bAngularSpring=true
    bOverrideAngularDamping=true
    bUseGravityOverride=true
    gravityScale=0.35
    simulationSpace=RootRelative
    alphaBoolBlend {blendOption=Linear}
    alphaScaleBiasClamp {bInterpResult=true}
    boundBone {boneName=dyn_R_belt_B2}
    componentLinearAccScale=<1 1 1>
    componentLinearVelScale=<1 1 1>
    gravityOverride {z=-2000}
    physicsBodyDefinitions [{
      boundBone {boneName=dyn_R_belt_B2}
      boxExtents=<30 30 30>
      constraintSetup {angularLimitsMax=<0 5 0>, angularLimitsMin=<0 -5 -10>, angularTarget=<1 0 0>}
      localJointOffset=<3 0 0>
     }]
    relativeSpaceBone {boneName=lowerarm_l}
   }
   n48 = Anim Dynamics - Bone: dyn_L_belt(ComponentPose=@n47, Alpha=1, ExternalForce="0,0,0") {
    angularDampingOverride=0.9999
    angularSpringConstant=2000
    bAngularSpring=true
    bOverrideAngularDamping=true
    bUseGravityOverride=true
    gravityScale=0.35
    simulationSpace=RootRelative
    alphaBoolBlend {blendOption=Linear}
    alphaScaleBiasClamp {bInterpResult=true}
    boundBone {boneName=dyn_L_belt}
    componentLinearAccScale=<1 1 1>
    componentLinearVelScale=<1 1 1>
    gravityOverride {z=-2000}
    physicsBodyDefinitions [{
      boundBone {boneName=dyn_L_belt}
      boxExtents=<30 30 30>
      constraintSetup {angularLimitsMax=<0 5 5>, angularLimitsMin=<0 -45 -10>, angularTarget=<1 0 0>}
      localJointOffset=<3 0 0>
     }]
    relativeSpaceBone {boneName=lowerarm_l}
   }
   n49 = "Transform (Modify) Bone - Bone: dyn_L_belt"(
    ComponentPose=@n48
    Alpha=Get isCrouching() -> (isCrouching:double)
    Translation="-1.782,-2.066,-0.389"
    Rotation="5.717,-5.216,3.649"
    Scale="1,1,1"
   ) {
    rotationMode=BMM_Additive
    translationMode=BMM_Additive
    translationSpace=BCS_BoneSpace
    alphaBoolBlend {blendOption=Linear}
    boneToModify {boneName=dyn_L_belt}
    rotation=<rot 5.717 -5.216 3.649>
    translation=<-1.782 -2.066 -0.389>
   }
   n50 = "Transform (Modify) Bone - Bone: dyn_front_strap_A_1"(ComponentPose=@n49, Alpha=1, Translation="0,0,-2", Rotation="0,0,-25", Scale="1,1,1") {
    rotationMode=BMM_Additive
    translationMode=BMM_Additive
    translationSpace=BCS_BoneSpace
    alphaBoolBlend {blendOption=Linear}
    boneToModify {boneName=dyn_front_strap_A_1}
    rotation {roll=-25}
    translation {z=-2}
   }
   n51 = "Constraint - dyn_front_strap_A_1 (P)"(ComponentPose=@n50, Alpha=0.5, ConstraintWeights_0=1) {
    alphaBoolBlend {blendOption=Linear}
    boneToModify {boneName=dyn_front_strap_A_1}
    constraintSetup [{targetBone {boneName=thigh_l}}]
    constraintWeights [1]
   }
   n52 = "Constraint - dyn_front_strap_A_2 (P)"(ComponentPose=@n51, Alpha=0.5, ConstraintWeights_0=1) {
    alphaBoolBlend {blendOption=Linear}
    boneToModify {boneName=dyn_front_strap_A_2}
    constraintSetup [{targetBone {boneName=dyn_L_frontSkirt_2}}]
    constraintWeights [1]
   }
   n53 = "Transform (Modify) Bone - Bone: dyn_beltConstraint_jnt_B"(ComponentPose=@n52, Alpha=1, Translation="0.000,0.805,0", Rotation="0.000,0.000,-8.787", Scale="1,1,1") {
    rotationMode=BMM_Additive
    translationMode=BMM_Additive
    alphaBoolBlend {blendOption=Linear}
    boneToModify {boneName=dyn_beltConstraint_jnt_B}
    rotation {roll=-8.787}
    translation {y=0.805}
   }
   n54 = "Transform (Modify) Bone - Bone: dyn_front_strap_B_1"(
    ComponentPose=@n53
    Alpha=Get FrontEnd_Idle_On_float() -> (FrontEnd_Idle_On_float:double)
    Translation="0,0,0"
    Rotation="0,0.000,5"
    Scale="1,1,1"
   ) {rotationMode=BMM_Additive, alphaBoolBlend {blendOption=Linear}, boneToModify {boneName=dyn_front_strap_B_1}, rotation {roll=5}}
   n55 = "Transform (Modify) Bone - Bone: dyn_hiddenBlade"(ComponentPose=@n54, Alpha=1, Translation="-14,0,0", Rotation="0,0,0", Scale="0.4,1,1") {
    scaleMode=BMM_Additive
    scaleSpace=BCS_BoneSpace
    translationMode=BMM_Additive
    translationSpace=BCS_BoneSpace
    alphaBoolBlend {blendOption=Linear}
    boneToModify {boneName=dyn_hiddenBlade}
    scale {x=0.4}
    translation {x=-14}
   }
   Output Pose(Result=Component To Local(ComponentPose=@n55)) {layerGroup=None}
  ]}
 EventGraph {nodes [
   Event BlueprintInitializeAnimation()
   > n1 = Set Skel_Mesh(Skel_Mesh=GetOwnerSkeletalMeshComponent())
   > Is Valid(InputObject=@n1)
   > Set Anim_Inst(Anim_Inst:Anim Instance Object Reference=GetAnimInstance(self=@n1)) -> (Output_Get:Anim Instance Object Reference)
   n2 = Get Anim_Inst() -> (Anim_Inst:Anim Instance Object Reference)
   > Is Valid(InputObject=@n2)
   > Branch(Condition=Get bIsInFrontEnd())
   n3 = Get bIsPlayingEmote()
   > n4 = Set FrontEnd_Idle_On(FrontEnd_Idle_On:bool=AND Boolean(
     A=Montage_IsPlaying(self=@n2, Montage=/BRCosmetics/Animation/Game/MainPlayer/Menu/BR/CustomSkins/_s20/JourneyMentorF/JourneyMentor_Female_Idle_Montage)
     B=NOT Boolean(A=@n3)
    )) -> (Output_Get:bool)
   > Set FrontEnd_Idle_On_float(FrontEnd_Idle_On_float:double="To Float (Boolean)"(InBool=@n4)) -> (Output_Get:double)
   Event BlueprintUpdateAnimation()
   > n5 = Sequence()
   > n6 = Branch(Condition=Get isZiplining() -> (isZiplining:bool))
   > Set ziplineDynamics(ziplineDynamics:double=0) -> (Output_Get:double)
   Set ziplineDynamics(execute=@n6.else, ziplineDynamics=1)
   Set isSkydiving(execute=@n5.then_1, isSkydiving:double="To Float (Boolean)"(InBool=Get bIsSkyDiving())) -> (Output_Get:double)
   n7 = Get bIsAccelerating2D()
   n8 = NOT Boolean(A=@n7)
   n9 = Get bIsCrouching()
   n10 = NOT Boolean(A=@n9)
   > Set isNotAcceleratingStanding(isNotAcceleratingStanding:double="To Float (Boolean)"(InBool=AND Boolean(A=@n8, B=@n10))) -> (Output_Get:double)
   > Set isCrouchMoving(isCrouchMoving:double="To Float (Boolean)"(InBool=AND Boolean(A=@n9, B=@n7))) -> (Output_Get:double)
   > Set isCrouching(isCrouching:double="To Float (Boolean)"(InBool=@n9)) -> (Output_Get:double)
   > Set crouchRBANweight(crouchRBANweight:double=SelectFloat(A=0.4, B=1, bPickA=@n9)) -> (Output_Get:double)
   n11 = Get bIsJumping()
   n12 = Get bIsFalling()
   > Set isJumping(isJumping:double="To Float (Boolean)"(InBool=OR Boolean(A=@n11, B=@n12))) -> (Output_Get:double)
   n13 = Event BlueprintProcessAnimInputEvent()
   > Set isZiplining(isZiplining:bool=IsNamed(self=@n13.AnimInputEvent, InEventName=aie.Player.Zipline.Entered)) -> (Output_Get:bool)
  ]}
]
variables [
 Anim_Inst {type=AnimInstance, default=null}
 crouchRBANweight {type=number}
 FrontEnd_Idle_On {type=boolean}
 FrontEnd_Idle_On_float {type=number}
 isCrouching {type=number}
 isCrouchMoving {type=number}
 isJumping {type=number}
 isNotAcceleratingStanding {type=number}
 isSkydiving {type=number}
 isZiplining {type=boolean}
 Skel_Mesh {type=SkeletalMeshComponent}
 ziplineDynamics {type=number}
]
