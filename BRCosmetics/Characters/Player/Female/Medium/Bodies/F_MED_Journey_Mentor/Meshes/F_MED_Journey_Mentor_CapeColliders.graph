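# Physics asset of capsule collision bodies for the F_MED_Journey_Mentor female player body, named as colliders for its cape.
# Capsules cover head, neck, upper spine, pelvis, clavicles, upper arms, legs and feet, with a second capsule on each thigh.
# The dynamic back bag and left and right back skirt bones also carry capsules.
# Constraints link bodies along the skeleton, many with free swing and twist, and attach bag and skirt bodies to the pelvis.
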PhysicsAsset /BRCosmetics/Characters/Player/Female/Medium/Bodies/F_MED_Journey_Mentor/Meshes/F_MED_Journey_Mentor_CapeColliders
bodies {
 calf_l [{shapeType=Capsule, length=34.6826, radius=9.482, center=<-19.9486 -2.8284 0.1383>, rotation=<rot 0 -91.1589 -92.0231>}]
 calf_r [{shapeType=Capsule, length=34.6826, radius=9.482, center=<19.998 2.8274 0.0937>, rotation=<rot 0 88.8411 -87.0231>}]
 clavicle_l [{shapeType=Capsule, length=6.6964, radius=7.3853, center=<3.9138 -0.8097 -2.9272>, rotation=<rot -88.3257 102.4929 82.4543>}]
 clavicle_r [{shapeType=Capsule, length=6.6964, radius=7.3853, center=<-3.4752 0.5305 3.3858>, rotation=<rot -88.3282 -77.5052 82.4562>}]
 dyn_L_backSkirt_2 [{shapeType=Capsule, length=10, radius=10, shapeName=dyn_R_backSkirt_2_capsule, center=<3 4 0>, rotation=<rot 90 0 0>}]
 dyn_R_backSkirt_2 [{shapeType=Capsule, length=10, radius=10, shapeName=dyn_R_backSkirt_2_capsule, center=<-3 -4 0>, rotation=<rot 90 0 0>}]
 dyn_backBag [{
   shapeType=Capsule
   length=8.088
   radius=7
   shapeName=dyn_backBag_capsule
   center=<8.9741 -0.6868 -0.5811>
   rotation=<rot 0 -88.5893 -82.9612>
  }]
 foot_l [{shapeType=Capsule, length=12.5982, radius=6.5127, center=<-2.9101 7.2748 -0.3501>, rotation=<rot 0 5.1558 -89.3972>}]
 foot_r [{shapeType=Capsule, length=12.6052, radius=6.5119, center=<2.9081 -7.2701 0.3337>, rotation=<rot 0 5.1452 -89.0835>}]
 head [{shapeType=Capsule, length=0, radius=18, center=<9.0697 0.0746 0>, rotation=<rot 0 -90 0>}]
 neck_02 [{shapeType=Capsule, length=0.505, radius=10.755, center=<-2.9035 0.6558 0>, rotation=<rot 0 -90 0>}]
 pelvis [{shapeType=Capsule, length=5.809, radius=14.3783, center=<-4.524 -1.8634 -0.1967>, rotation=<rot -4.1601 -89.9995 -1.4037>}]
 spine_03 [{shapeType=Capsule, length=12.1237, radius=13.8935, center=<-4.5611 1.7894 -0.1524>, rotation=<rot 6.2753 -83.7246 89.9992>}]
 spine_04 [{shapeType=Capsule, length=17.1369, radius=16, shapeName=spine_04_capsule, center=<-4.6639 1.1618 1.2345>, rotation=<rot 0 -90 0>}]
 spine_05 [{shapeType=Capsule, length=17.2046, radius=12.0791, center=<-1.9814 1.858 0.7467>, rotation=<rot 0 -90 0>}]
 thigh_l [
  {shapeType=Capsule, length=30.6477, radius=10.4942, center=<-22.3366 -1.2788 -0.1036>, rotation=<rot 0 92.6888 -89.6929>}
  {shapeType=Capsule, length=12.25, radius=8.5, center=<-13.3119 -0.0001 -8.2693>, rotation=<rot -90 0 -15.0002>}
 ]
 thigh_r [
  {shapeType=Capsule, length=30.6477, radius=10.4942, center=<22.3407 1.2789 0.6548>, rotation=<rot 0 -87.3111 -89.6929>}
  {shapeType=Capsule, length=13, radius=8, center=<13.8792 -0.1751 9.8313>, rotation=<rot 90 14.0363 38.0889>}
 ]
 upperarm_l [{shapeType=Capsule, length=19.4828, radius=7.4326, center=<13.6692 0.2831 0.4097>, rotation=<rot 0.2555 91.1854 -88.5506>}]
 upperarm_r [{shapeType=Capsule, length=16.8789, radius=7.1408, center=<-12.334 0.1139 -0.0744>, rotation=<rot 0 -90 -84.8383>}]
}
constraints [
 {bone1Name=spine_03, bone2Name=pelvis, swing1LimitDegrees=45, swing2LimitDegrees=45, twistLimitDegrees=45}
 {bone1Name=spine_05, bone2Name=spine_04, swing1LimitDegrees=45, swing2LimitDegrees=45, twistLimitDegrees=45}
 {bone1Name=upperarm_l, bone2Name=clavicle_l, swing1LimitDegrees=45, swing2LimitDegrees=45, twistLimitDegrees=45}
 {bone1Name=upperarm_r, bone2Name=clavicle_r, swing1LimitDegrees=45, swing2LimitDegrees=45, twistLimitDegrees=45}
 {bone1Name=thigh_l, bone2Name=pelvis, swing1LimitDegrees=45, swing2LimitDegrees=45, twistLimitDegrees=45}
 {bone1Name=calf_l, bone2Name=thigh_l, swing1LimitDegrees=45, swing2LimitDegrees=45, twistLimitDegrees=45}
 {bone1Name=thigh_r, bone2Name=pelvis, swing1LimitDegrees=45, swing2LimitDegrees=45, twistLimitDegrees=45}
 {bone1Name=calf_r, bone2Name=thigh_r, swing1LimitDegrees=45, swing2LimitDegrees=45, twistLimitDegrees=45}
 {
  bone1Name=clavicle_r
  bone2Name=spine_05
  swing1LimitDegrees=45
  swing1Motion=Free
  swing2LimitDegrees=45
  swing2Motion=Free
  twistLimitDegrees=45
  twistMotion=Free
 }
 {
  bone1Name=foot_l
  bone2Name=calf_l
  swing1LimitDegrees=45
  swing1Motion=Free
  swing2LimitDegrees=45
  swing2Motion=Free
  twistLimitDegrees=45
  twistMotion=Free
 }
 {
  bone1Name=foot_r
  bone2Name=calf_r
  swing1LimitDegrees=45
  swing1Motion=Free
  swing2LimitDegrees=45
  swing2Motion=Free
  twistLimitDegrees=45
  twistMotion=Free
 }
 {
  bone1Name=clavicle_l
  bone2Name=spine_05
  swing1LimitDegrees=45
  swing1Motion=Free
  swing2LimitDegrees=45
  swing2Motion=Free
  twistLimitDegrees=45
  twistMotion=Free
 }
 {
  bone1Name=neck_02
  bone2Name=spine_05
  swing1LimitDegrees=45
  swing1Motion=Free
  swing2LimitDegrees=45
  swing2Motion=Free
  twistLimitDegrees=45
  twistMotion=Free
 }
 {
  bone1Name=head
  bone2Name=neck_02
  swing1LimitDegrees=45
  swing1Motion=Free
  swing2LimitDegrees=45
  swing2Motion=Free
  twistLimitDegrees=45
  twistMotion=Free
 }
 {
  bone1Name=dyn_backBag
  bone2Name=pelvis
  swing1LimitDegrees=45
  swing1Motion=Free
  swing2LimitDegrees=45
  swing2Motion=Free
  twistLimitDegrees=45
  twistMotion=Free
 }
 {
  bone1Name=spine_04
  bone2Name=spine_03
  swing1LimitDegrees=45
  swing1Motion=Free
  swing2LimitDegrees=45
  swing2Motion=Free
  twistLimitDegrees=45
  twistMotion=Free
 }
 {
  bone1Name=dyn_L_backSkirt_2
  bone2Name=pelvis
  swing1LimitDegrees=45
  swing1Motion=Free
  swing2LimitDegrees=45
  swing2Motion=Free
  twistLimitDegrees=45
  twistMotion=Free
 }
 {
  bone1Name=dyn_R_backSkirt_2
  bone2Name=pelvis
  swing1LimitDegrees=45
  swing1Motion=Free
  swing2LimitDegrees=45
  swing2Motion=Free
  twistLimitDegrees=45
  twistMotion=Free
 }
]
