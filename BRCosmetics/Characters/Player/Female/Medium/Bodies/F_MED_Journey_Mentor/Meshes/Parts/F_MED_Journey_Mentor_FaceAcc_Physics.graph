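# Physics asset for the face accessory part of the female Journey Mentor character cosmetic, giving physics bodies to a left front braid and a back braid.
# Each braid is a four-capsule chain constrained link by link up to the head, which joins spine_05 with free angular motion.
# The back braid allows wide swing with minimal twist, while the front braid swings more tightly and locks twist at its second link.
# Larger capsules approximate the head and upper torso.
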PhysicsAsset /BRCosmetics/Characters/Player/Female/Medium/Bodies/F_MED_Journey_Mentor/Meshes/Parts/F_MED_Journey_Mentor_FaceAcc_Physics
bodies {
 dyn_L_front_braid_1 [{shapeType=Capsule, length=5, radius=1.6, shapeName=dyn_L_front_braid_1_capsule, center=<5 0 0>, rotation=<rot 90 0 0>}]
 dyn_L_front_braid_2 [{shapeType=Capsule, length=2.25, radius=1.3, shapeName=dyn_L_front_braid_2_capsule, center=<3.6478 0 0>, rotation=<rot 90 0 0>}]
 dyn_L_front_braid_3 [{shapeType=Capsule, length=3.5, radius=1.5, shapeName=dyn_L_front_braid_3_capsule, center=<1.8987 0 0>, rotation=<rot 90 0 0>}]
 dyn_L_front_braid_4 [{shapeType=Capsule, length=2.75, radius=3.75, shapeName=dyn_L_front_braid_3_capsule, center=<4.3223 0 0>, rotation=<rot 90 0 0>}]
 dyn_back_braid_1 [{shapeType=Capsule, length=5.75, radius=1.3, shapeName=dyn_back_braid_1_capsule, center=<5 0 0>, rotation=<rot 90 0 0>}]
 dyn_back_braid_2 [{shapeType=Capsule, length=5.5, radius=1.3, shapeName=dyn_back_braid_2_capsule, center=<5 0 0>, rotation=<rot 90 0 0>}]
 dyn_back_braid_3 [{shapeType=Capsule, length=5.25, radius=1.2, shapeName=dyn_back_braid_3_capsule, center=<4 0 0>, rotation=<rot 90 0 0>}]
 dyn_back_braid_4 [{shapeType=Capsule, length=4, radius=1.8, shapeName=dyn_back_braid_4_capsule, center=<5 0 0.5>, rotation=<rot 90 0 0>}]
 head [{shapeType=Capsule, length=5, radius=7, shapeName=head_capsule, center=<4.2507 -3.2083 0>, rotation=<rot 90 0 -12.9419>}]
 spine_05 [
  {shapeType=Capsule, length=31.25, radius=13.75, shapeName=spine_05_capsule}
  {shapeType=Capsule, length=9, radius=9, shapeName=spine_05_capsule, center=<12.3674 -9.6937 0>}
  {shapeType=Capsule, length=15, radius=15, shapeName=spine_05_capsule, center=<0 -5.4001 0>}
 ]
}
constraints [
 {bone1Name=dyn_back_braid_1, bone2Name=head, swing1LimitDegrees=20, swing2LimitDegrees=30, twistLimitDegrees=1}
 {bone1Name=dyn_back_braid_2, bone2Name=dyn_back_braid_1, swing1LimitDegrees=20, swing2LimitDegrees=70, twistLimitDegrees=1}
 {bone1Name=dyn_back_braid_3, bone2Name=dyn_back_braid_2, swing1LimitDegrees=30, swing2LimitDegrees=70, twistLimitDegrees=1}
 {bone1Name=dyn_back_braid_4, bone2Name=dyn_back_braid_3, swing1LimitDegrees=20, swing2LimitDegrees=70, twistLimitDegrees=1}
 {bone1Name=dyn_L_front_braid_1, bone2Name=head, swing1LimitDegrees=5, swing2LimitDegrees=5, twistLimitDegrees=5}
 {
  bone1Name=dyn_L_front_braid_2
  bone2Name=dyn_L_front_braid_1
  swing1LimitDegrees=10
  swing2LimitDegrees=15
  twistLimitDegrees=45
  twistMotion=Locked
 }
 {bone1Name=dyn_L_front_braid_3, bone2Name=dyn_L_front_braid_2, swing1LimitDegrees=15, swing2LimitDegrees=15, twistLimitDegrees=5}
 {
  bone1Name=head
  bone2Name=spine_05
  swing1LimitDegrees=45
  swing1Motion=Free
  swing2LimitDegrees=45
  swing2Motion=Free
  twistLimitDegrees=45
  twistMotion=Free
 }
 {bone1Name=dyn_L_front_braid_4, bone2Name=dyn_L_front_braid_3, swing1LimitDegrees=15, swing2LimitDegrees=15, twistLimitDegrees=5}
]
