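# Post-processes the Journey Mentor female face accessory mesh, adding braid and hood motion to the pose copied from its attached parent.
# Two branches each constrain the left front braid and run RigidBody physics with separate physics assets, blended by whether a backpack is worn without an emote playing.
# The backpack branch also applies alpha-driven back braid corrections for arm raise, crouch movement, skydiving and a general braid transform.
# HeadAlpha-driven hood bone offsets follow the blend.
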AnimBlueprint /BRCosmetics/Characters/Player/Female/Medium/Bodies/F_MED_Journey_Mentor/Meshes/Parts/F_MED_Journey_Mentor_FaceAcc_AnimBP
target_skeleton=&/F_MED_Journey_Mentor_FaceAcc_Skeleton
graphs [
 AnimGraph {nodes [
   n1 = "Constraint - dyn_L_front_braid_1 (PPP)"!(
    ComponentPose=Local To Component(LocalPose=Copy Pose From Mesh() {bUseAttachedParent=true})
    Alpha=1
    ConstraintWeights_0=0.75
    ConstraintWeights_1:float=0.25
    ConstraintWeights_2:float=0
   ) -> (Pose) {
    alphaBoolBlend {blendOption=Linear}
    boneToModify {boneName=dyn_L_front_braid_1}
    constraintSetup [{targetBone {boneName=neck_01}}, {targetBone {boneName=spine_05}}, {targetBone {boneName=head}}]
    constraintWeights [1,1,1]
   }
   n2 = RigidBody!(ComponentPose=@n1, bAlphaBoolEnabled:bool, OverrideWorldGravity:Vector="0,0,-700", ExternalForce) -> (Pose) {
    alphaInputType=Bool
    bClampLinearTranslationLimitToRefPose=true
    bOverrideWorldGravity=true
    overridePhysicsAsset=&/F_MED_Journey_Mentor_FaceAcc_Backpack_Physics
    simulationSpace=BaseBoneSpace
    alphaBoolBlend {blendOption=Linear}
    baseBoneRef {boneName=neck_02}
    componentAppliedLinearAccClamp=<2000 3000 1000>
    componentLinearAccScale=<0.25 0.2 0.05>
    componentLinearVelScale=<0.25 0.2 0.05>
    overrideWorldGravity {z=-700}
   }
   n3 = Get UpperArmLRotUpAlpha() -> (UpperArmLRotUpAlpha:float)
   n4 = "Transform (Modify) Bone - Bone: dyn_back_braid_1"(ComponentPose=@n2, Alpha=@n3, Translation="0,0,0", Rotation="15,0.000,-0.000", Scale="1,1,1") {
    rotationMode=BMM_Additive
    rotationSpace=BCS_BoneSpace
    translationMode=BMM_Additive
    translationSpace=BCS_BoneSpace
    alphaBoolBlend {blendOption=Linear}
    boneToModify {boneName=dyn_back_braid_1}
    rotation {pitch=15}
   }
   n5 = "Transform (Modify) Bone - Bone: dyn_back_braid_2"(ComponentPose=@n4, Alpha=@n3, Translation="0,0,0", Rotation="-30,0,-0.000", Scale="1,1,1") {
    rotationMode=BMM_Additive
    rotationSpace=BCS_BoneSpace
    translationMode=BMM_Additive
    translationSpace=BCS_BoneSpace
    alphaBoolBlend {blendOption=Linear}
    boneToModify {boneName=dyn_back_braid_2}
    rotation {pitch=-30}
   }
   n6 = "Transform (Modify) Bone - Bone: dyn_back_braid_3"(ComponentPose=@n5, Alpha=@n3, Translation="0,0,0", Rotation="0,0,0", Scale="1,1,1") {
    rotationMode=BMM_Additive
    rotationSpace=BCS_BoneSpace
    translationMode=BMM_Additive
    translationSpace=BCS_BoneSpace
    alphaBoolBlend {blendOption=Linear}
    boneToModify {boneName=dyn_back_braid_3}
   }
   n7 = "Transform (Modify) Bone - Bone: dyn_back_braid_4"(ComponentPose=@n6, Alpha=@n3, Translation="0,0,0", Rotation="0,0,0", Scale="1,1,1") {
    rotationMode=BMM_Additive
    rotationSpace=BCS_BoneSpace
    translationMode=BMM_Additive
    translationSpace=BCS_BoneSpace
    alphaBoolBlend {blendOption=Linear}
    boneToModify {boneName=dyn_back_braid_4}
   }
   n8 = Get TransformBackBraidAlpha() -> (TransformBackBraidAlpha:float)
   n9 = "Transform (Modify) Bone - Bone: dyn_back_braid_1"(ComponentPose=@n7, Alpha=@n8, Translation="0,0,0", Rotation="0,0,0", Scale="0.85,1,1") {scaleMode=BMM_Additive, alphaBoolBlend {blendOption=Linear}, boneToModify {boneName=dyn_back_braid_1}, scale {x=0.85}}
   n10 = "Transform (Modify) Bone - Bone: dyn_back_braid_2"(ComponentPose=@n9, Alpha=@n8, Translation="0,0,0", Rotation="0,0,-5", Scale="0.9,1,1") {scaleMode=BMM_Additive, alphaBoolBlend {blendOption=Linear}, boneToModify {boneName=dyn_back_braid_2}, rotation {roll=-5}, scale {x=0.9}}
   n11 = Get IsCrouchMovingAlpha() -> (IsCrouchMovingAlpha:float)
   n12 = "Transform (Modify) Bone - Bone: dyn_back_braid_2"(ComponentPose=@n10, Alpha=@n11, Translation="0,0,0", Rotation="0,0,-7", Scale="1,1,1") {rotationMode=BMM_Additive, alphaBoolBlend {blendOption=Linear}, boneToModify {boneName=dyn_back_braid_2}, rotation {roll=-7}}
   n13 = "Transform (Modify) Bone - Bone: dyn_back_braid_3"(ComponentPose=@n12, Alpha=@n11, Translation="0,0,0", Rotation="0,0,-8", Scale="1,1,1") {rotationMode=BMM_Additive, alphaBoolBlend {blendOption=Linear}, boneToModify {boneName=dyn_back_braid_3}, rotation {roll=-8}}
   n14 = "Transform (Modify) Bone - Bone: dyn_back_braid_3"(ComponentPose=@n13, Alpha=@n11, Translation="0,0,0", Rotation="0,0,-5", Scale="1,1,1") {rotationMode=BMM_Additive, alphaBoolBlend {blendOption=Linear}, boneToModify {boneName=dyn_back_braid_3}, rotation {roll=-5}}
   n15 = Get IsSkyDivingAlpha() -> (IsSkyDivingAlpha:float)
   n16 = "Transform (Modify) Bone - Bone: dyn_back_braid_1"(ComponentPose=@n14, Alpha=@n15, Translation="0,0,0", Rotation="0,0,30", Scale="1,1,1") {rotationMode=BMM_Additive, alphaBoolBlend {blendOption=Linear}, boneToModify {boneName=dyn_back_braid_1}, rotation {roll=30}}
   n17 = "Transform (Modify) Bone - Bone: dyn_back_braid_2"(ComponentPose=@n16, Alpha=@n15, Translation="0,0,0", Rotation="0,0,10", Scale="1,1,1") {rotationMode=BMM_Additive, alphaBoolBlend {blendOption=Linear}, boneToModify {boneName=dyn_back_braid_2}, rotation {roll=10}}
   n18 = Component To Local(ComponentPose=@n17)
   n19 = "Constraint - dyn_L_front_braid_1 (PPP)"(
    ComponentPose=Local To Component(LocalPose=Copy Pose From Mesh() {bUseAttachedParent=true})
    Alpha=1
    ConstraintWeights_0=0.75
    ConstraintWeights_1=0.25
    ConstraintWeights_2=0
   ) {
    alphaBoolBlend {blendOption=Linear}
    boneToModify {boneName=dyn_L_front_braid_1}
    constraintSetup [{targetBone {boneName=neck_01}}, {targetBone {boneName=spine_05}}, {targetBone {boneName=head}}]
    constraintWeights [1,1,1]
   }
   n20 = RigidBody!(ComponentPose=@n19, Alpha:float, ExternalForce) -> (Pose) {
    bClampLinearTranslationLimitToRefPose=true
    bOverrideWorldGravity=true
    overridePhysicsAsset=&/F_MED_Journey_Mentor_FaceAcc_Physics
    simulationSpace=BaseBoneSpace
    alphaBoolBlend {blendOption=Linear}
    baseBoneRef {boneName=neck_02}
    componentAppliedLinearAccClamp=<3000 3000 3000>
    componentLinearAccScale=<0.25 0.2 0.25>
    componentLinearVelScale=<0.25 0.2 0.25>
    overrideWorldGravity {z=-500}
   }
   n21 = Component To Local(ComponentPose=@n20)
   n22 = Get bIsWearingBackpackAndNotPlayingEmote() -> (bIsWearingBackpackAndNotPlayingEmote:bool)
   n23 = Get HeadAlpha() -> (HeadAlpha:float)
   n24 = "Transform (Modify) Bone - Bone: hood_dyn"(
    ComponentPose=Local To Component(LocalPose=Blend Poses by bool(BlendPose_0=@n18, BlendPose_1=@n21, bActiveValue=@n22) {blendType=Linear})
    Alpha=@n23
    Translation="0,0,0"
    Rotation="-0.015,-20,-0.006"
    Scale="1,1,1"
   ) {
    rotationMode=BMM_Additive
    rotationSpace=BCS_BoneSpace
    alphaBoolBlend {blendOption=Linear}
    boneToModify {boneName=hood_dyn}
    rotation=<rot -0.015 -20 -0.006>
   }
   n25 = "Transform (Modify) Bone - Bone: hood_r"(ComponentPose=@n24, Alpha=@n23, Translation="0.000,0.000,-3", Rotation="0,0,0", Scale="1,1,1") {
    rotationMode=BMM_Additive
    rotationSpace=BCS_BoneSpace
    translationMode=BMM_Additive
    translationSpace=BCS_BoneSpace
    alphaBoolBlend {blendOption=Linear}
    boneToModify {boneName=hood_r}
    translation {z=-3}
   }
   Output Pose(Result=Component To Local(ComponentPose=@n25)) {layerGroup=None}
  ]}
 EventGraph {links [], nodes []}
]
variables []
