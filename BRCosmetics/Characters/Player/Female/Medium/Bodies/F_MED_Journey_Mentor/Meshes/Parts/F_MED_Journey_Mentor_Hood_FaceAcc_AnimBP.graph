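# Animation Blueprint for the F_MED_Journey_Mentor hood face accessory, adding secondary motion and pose-driven correctives to its hood bones.
# It copies the attached parent's pose, then runs spring-based Anim Dynamics on the left, right and back hood bones.
# Each update, head yaw fades back-hood dynamics out as an additive offset fades in, while left upper-arm pitch drives a second offset.
# Left-arm yaw alpha is computed but unused in the AnimGraph.
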AnimBlueprint /BRCosmetics/Characters/Player/Female/Medium/Bodies/F_MED_Journey_Mentor/Meshes/Parts/F_MED_Journey_Mentor_Hood_FaceAcc_AnimBP
target_skeleton=&/F_MED_Journey_Mentor_Hood_FaceAcc_Skeleton
graphs [
 AnimGraph {nodes [
   n1 = Anim Dynamics - Bone: dyn_L_front_hood(ComponentPose=Local To Component(LocalPose=Copy Pose From Mesh() {bUseAttachedParent=true}), Alpha=1, ExternalForce="0,0,0") {
    angularDampingOverride=1
    angularSpringConstant=20000
    bAngularSpring=true
    bOverrideAngularDamping=true
    bUseGravityOverride=true
    gravityScale=0.35
    simulationSpace=RootRelative
    alphaBoolBlend {blendOption=Linear}
    boundBone {boneName=dyn_L_front_hood}
    componentAppliedLinearAccClamp=<3000 3000 3000>
    componentLinearAccScale=<0.25 0.25 0.25>
    componentLinearVelScale=<0.25 0.25 0.25>
    gravityOverride {z=-980}
    physicsBodyDefinitions [{
      boundBone {boneName=dyn_L_front_hood}
      boxExtents=<10 10 10>
      constraintSetup {angularLimitsMax=<0 0 0>, angularLimitsMin=<0 0 0>, angularTarget=<0 0 0>, linearAxesMax=<2 2 2>, linearAxesMin=<0 -2 -2>}
      localJointOffset=<0 0 0>
     }]
   }
   n2 = Anim Dynamics - Bone: dyn_R_front_hood(ComponentPose=@n1, Alpha=1, ExternalForce="0,0,0") {
    angularDampingOverride=1
    angularSpringConstant=20000
    bAngularSpring=true
    bOverrideAngularDamping=true
    bUseGravityOverride=true
    gravityScale=0.35
    simulationSpace=RootRelative
    alphaBoolBlend {blendOption=Linear}
    boundBone {boneName=dyn_R_front_hood}
    componentAppliedLinearAccClamp=<3000 3000 3000>
    componentLinearAccScale=<0.25 0.25 0.25>
    componentLinearVelScale=<0.25 0.25 0.25>
    gravityOverride {z=-980}
    physicsBodyDefinitions [{
      boundBone {boneName=dyn_R_front_hood}
      boxExtents=<10 10 10>
      constraintSetup {angularLimitsMax=<0 0 0>, angularLimitsMin=<0 0 0>, angularTarget=<0 0 0>, linearAxesMax=<0 2 2>, linearAxesMin=<-2 -2 -2>}
      localJointOffset=<0 0 0>
     }]
   }
   n3 = Anim Dynamics - Bone: dyn_back_hood(ComponentPose=@n2, Alpha=Get head_alpha_inverse() -> (head_alpha_inverse:double), ExternalForce="0,0,0") {
    angularDampingOverride=1
    angularSpringConstant=2000
    bAngularSpring=true
    bOverrideAngularDamping=true
    bUseGravityOverride=true
    gravityScale=0.35
    simulationSpace=RootRelative
    alphaBoolBlend {blendOption=Linear}
    boundBone {boneName=dyn_back_hood}
    componentAppliedLinearAccClamp=<1500 1500 1500>
    componentLinearAccScale=<0.2 0.2 0.2>
    componentLinearVelScale=<0.2 0.2 0.2>
    gravityOverride {z=-490}
    physicsBodyDefinitions [{
      boundBone {boneName=dyn_back_hood}
      boxExtents=<10 10 10>
      constraintSetup {angularLimitsMax=<0 0 0>, angularLimitsMin=<0 0 0>, angularTarget=<0 0 0>, linearAxesMax=<2 2 2>, linearAxesMin=<-2 -2 0>}
      localJointOffset=<0 5 0>
     }]
   }
   n4 = "Transform (Modify) Bone - Bone: dyn_back_hood"(ComponentPose=@n3, Alpha=Get head_alpha() -> (head_alpha:double), Translation="0,0,3", Rotation="0,0,0", Scale="1,1,1") {
    translationMode=BMM_Additive
    translationSpace=BCS_BoneSpace
    alphaBoolBlend {blendOption=Linear}
    boneToModify {boneName=dyn_back_hood}
    translation {z=3}
   }
   n5 = "Transform (Modify) Bone - Bone: dyn_back_hood"(
    ComponentPose=@n4
    Alpha=Get armUpLeftAlpha() -> (armUpLeftAlpha:double)
    Translation="-5.404,-0.000,-2.029"
    Rotation="0,0,0"
    Scale="1,1,1"
   ) {
    translationMode=BMM_Additive
    translationSpace=BCS_BoneSpace
    alphaBoolBlend {blendOption=Linear}
    boneToModify {boneName=dyn_back_hood}
    translation {x=-5.404, z=-2.029}
   }
   Output Pose(Result=Component To Local(ComponentPose=@n5)) {layerGroup=None}
  ]}
 EventGraph {nodes [
   Event BlueprintUpdateAnimation()
   n1 = BreakTransform(InTransform=GetDeltaTransformFromRefPose(self=GetOwnerSkeletalMeshComponent(), BoneName=head, BaseName=neck_02))
   > n2 = Set head_alpha(head_alpha:double=MapRangeClamped(Value=BreakRotator(InRot=@n1.Rotation).Yaw, InRangeA=0, InRangeB=-20)) -> (Output_Get:double)
   > Set head_alpha_inverse(head_alpha_inverse:double=float - float(A=1, B=@n2)) -> (Output_Get:double)
   n3 = GetOwnerSkeletalMeshComponent()
   n4 = BreakRotator(InRot=BreakTransform(InTransform=GetDeltaTransformFromRefPose(self=@n3, BoneName=upperarm_l, BaseName=clavicle_l)).Rotation)
   > Set armUpLeftAlpha(armUpLeftAlpha:double=MapRangeClamped(Value=@n4.Pitch, InRangeA=0, InRangeB=40)) -> (Output_Get:double)
   > Set armYawLeftAlpha(armYawLeftAlpha:double=MapRangeClamped(Value=@n4.Yaw, InRangeA=0, InRangeB=-30)) -> (Output_Get:double)
  ]}
]
variables [
 armUpLeftAlpha {type=number}
 armYawLeftAlpha {type=number}
 head_alpha {type=number}
 head_alpha_inverse {type=number}
]
